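# Physics asset for the female medium Pineapple Bandit body cosmetic, covering its dangling "smoke" pieces and shoestrings.
# Capsule bodies on pelvis, lower spine and feet form the base, joined by constraints with free swing and twist.
# Two smoke bones hang from base bodies linked to spine_02, and two shoestrings attach to each foot.
# These dynamic bones have tight swing and twist limits, keeping secondary motion small.
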PhysicsAsset /BRCosmetics/Characters/Player/Female/Medium/Bodies/F_MED_Pineapple_Bandit/Meshes/F_MED_Pineapple_Bandit_Physics
bodies {
 dyn_shoestring_a_l [{shapeType=Capsule, length=1, radius=1, center=<1.5 0 0>, rotation=<rot 90 0 0>}]
 dyn_shoestring_a_r [{shapeType=Capsule, length=1, radius=1, center=<-1.5 0 0>, rotation=<rot 90 0 180>}]
 dyn_shoestring_b_l [{shapeType=Capsule, length=1, radius=1, center=<1.5 0 0>, rotation=<rot 90 0 0>}]
 dyn_shoestring_b_r [{shapeType=Capsule, length=1, radius=1, center=<-1.5 0 0>, rotation=<rot 90 0 180>}]
 dyn_smoke_a [{shapeType=Capsule, length=6.3322, radius=2.8235, center=<0.644 1.9339 -0.4552>, rotation=<rot 0 88.7295 -85.6451>}]
 dyn_smoke_a_base [{shapeType=Capsule, length=1, radius=1}]
 dyn_smoke_b [{shapeType=Capsule, length=6.4079, radius=2.7109, center=<0.5628 1.8769 -0.2096>, rotation=<rot 0 87.7796 -85.587>}]
 dyn_smoke_b_base [{shapeType=Capsule, length=1, radius=1}]
 foot_l [{shapeType=Capsule, length=1, radius=1}]
 foot_r [{shapeType=Capsule, length=1, radius=1}]
 pelvis [{shapeType=Capsule, length=9, radius=10.5, center=<0.973 -1.6187 0>}]
 spine_01 [{shapeType=Capsule, length=5.25, radius=9.75, center=<0 0.3491 0>}]
 spine_02 [{shapeType=Capsule, length=0.25, radius=10.25, center=<0 0.8286 0>}]
}
constraints [
 {
  bone1Name=spine_01
  bone2Name=pelvis
  swing1LimitDegrees=45
  swing1Motion=Free
  swing2LimitDegrees=45
  swing2Motion=Free
  twistLimitDegrees=45
  twistMotion=Free
 }
 {
  bone1Name=spine_02
  bone2Name=spine_01
  swing1LimitDegrees=45
  swing1Motion=Free
  swing2LimitDegrees=45
  swing2Motion=Free
  twistLimitDegrees=45
  twistMotion=Free
 }
 {
  bone1Name=dyn_smoke_b_base
  bone2Name=spine_02
  swing1LimitDegrees=45
  swing1Motion=Free
  swing2LimitDegrees=45
  swing2Motion=Free
  twistLimitDegrees=45
  twistMotion=Free
 }
 {
  bone1Name=dyn_smoke_a_base
  bone2Name=spine_02
  swing1LimitDegrees=45
  swing1Motion=Free
  swing2LimitDegrees=45
  swing2Motion=Free
  twistLimitDegrees=45
  twistMotion=Free
 }
 {bone1Name=dyn_smoke_b, bone2Name=dyn_smoke_b_base, swing1LimitDegrees=15, swing2LimitDegrees=15, twistLimitDegrees=15}
 {bone1Name=dyn_smoke_a, bone2Name=dyn_smoke_a_base, swing1LimitDegrees=10, swing2LimitDegrees=10, twistLimitDegrees=10}
 {
  bone1Name=foot_r
  bone2Name=pelvis
  swing1LimitDegrees=45
  swing1Motion=Free
  swing2LimitDegrees=45
  swing2Motion=Free
  twistLimitDegrees=45
  twistMotion=Free
 }
 {
  bone1Name=foot_l
  bone2Name=pelvis
  swing1LimitDegrees=45
  swing1Motion=Free
  swing2LimitDegrees=45
  swing2Motion=Free
  twistLimitDegrees=45
  twistMotion=Free
 }
 {bone1Name=dyn_shoestring_a_l, bone2Name=foot_l, swing1LimitDegrees=10, swing2LimitDegrees=10, twistLimitDegrees=10}
 {bone1Name=dyn_shoestring_b_l, bone2Name=foot_l, swing1LimitDegrees=15, swing2LimitDegrees=15, twistLimitDegrees=15}
 {bone1Name=dyn_shoestring_a_r, bone2Name=foot_r, swing1LimitDegrees=15, swing2LimitDegrees=15, twistLimitDegrees=15}
 {bone1Name=dyn_shoestring_b_r, bone2Name=foot_r, swing1LimitDegrees=15, swing2LimitDegrees=15, twistLimitDegrees=15}
]
