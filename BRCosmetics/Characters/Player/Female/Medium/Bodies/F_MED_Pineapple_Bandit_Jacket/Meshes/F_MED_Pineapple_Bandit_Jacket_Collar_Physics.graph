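# Physics asset for the collar of the female medium Pineapple Bandit Jacket player cosmetic, defining bodies and joints for its dynamic collar bones.
# Five collar points (back, front left/right, side left/right) each pair a tiny capsule on a base bone with a larger, offset capsule on the flap bone.
# A constraint links each flap to its own base with moderate swing and tight twist limits, with no constraints between neighboring flaps.
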PhysicsAsset /BRCosmetics/Characters/Player/Female/Medium/Bodies/F_MED_Pineapple_Bandit_Jacket/Meshes/F_MED_Pineapple_Bandit_Jacket_Collar_Physics
bodies {
 dyn_collar_back [{shapeType=Capsule, length=5, radius=3, center=<4 0 0>, rotation=<rot 90 0 0>}]
 dyn_collar_back_base [{shapeType=Capsule, length=0.505, radius=0.505, rotation=<rot 0 -90 0>}]
 dyn_collar_front_base_l [{shapeType=Capsule, length=0.505, radius=0.505, rotation=<rot 0 -90 0>}]
 dyn_collar_front_base_r [{shapeType=Capsule, length=0.505, radius=0.505, rotation=<rot 0 -90 0>}]
 dyn_collar_front_l [{shapeType=Capsule, length=5, radius=3, center=<4 0 0>, rotation=<rot 90 0 0>}]
 dyn_collar_front_r [{shapeType=Capsule, length=5, radius=3, center=<-4 0 0>, rotation=<rot 90 0 0>}]
 dyn_collar_side_base_l [{shapeType=Capsule, length=0.505, radius=0.505, rotation=<rot 0 -90 0>}]
 dyn_collar_side_base_r [{shapeType=Capsule, length=0.505, radius=0.505, rotation=<rot 0 -90 0>}]
 dyn_collar_side_l [{shapeType=Capsule, length=5, radius=3, center=<4 0 0>, rotation=<rot 90 0 0>}]
 dyn_collar_side_r [{shapeType=Capsule, length=5, radius=3, center=<-4 0 0>, rotation=<rot 90 0 0>}]
}
constraints [
 {bone1Name=dyn_collar_front_l, bone2Name=dyn_collar_front_base_l, swing1LimitDegrees=25, swing2LimitDegrees=25, twistLimitDegrees=5}
 {bone1Name=dyn_collar_side_l, bone2Name=dyn_collar_side_base_l, swing1LimitDegrees=25, swing2LimitDegrees=25, twistLimitDegrees=5}
 {bone1Name=dyn_collar_front_r, bone2Name=dyn_collar_front_base_r, swing1LimitDegrees=25, swing2LimitDegrees=25, twistLimitDegrees=5}
 {bone1Name=dyn_collar_side_r, bone2Name=dyn_collar_side_base_r, swing1LimitDegrees=25, swing2LimitDegrees=25, twistLimitDegrees=5}
 {bone1Name=dyn_collar_back, bone2Name=dyn_collar_back_base, swing1LimitDegrees=25, swing2LimitDegrees=25, twistLimitDegrees=5}
]
