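# Physics asset for the female medium Pineapple Bandit jacket body, setting up secondary motion for the jacket, its two flaps and a spray can.
# Front, side and back three-bone chains use limited swing joints from base bodies attached to spine_04; the flaps also hang there, the can from thigh_l.
# Neighboring hem (C) bodies are cross-linked with mostly locked angular joints, tying the chains together except across the front opening.
# Pelvis, spine and thigh capsules form the body volume.
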PhysicsAsset /BRCosmetics/Characters/Player/Female/Medium/Bodies/F_MED_Pineapple_Bandit_Jacket/Meshes/F_MED_Pineapple_Bandit_Jacket_Physics
bodies {
 dyn_jacket_bk_A [{shapeType=Capsule, length=7, radius=5, shapeName=dyn_jacket_bk_A_capsule, center=<0 -4 0>}]
 dyn_jacket_bk_A_base [{shapeType=Capsule, length=1, radius=1}]
 dyn_jacket_bk_A_base_l [{shapeType=Capsule, length=1, radius=1}]
 dyn_jacket_bk_A_base_r [{shapeType=Capsule, length=1, radius=1}]
 dyn_jacket_bk_A_l [{shapeType=Capsule, length=7, radius=5, shapeName=dyn_jacket_bk_A_l_capsule, center=<0 -4 0>}]
 dyn_jacket_bk_A_r [{shapeType=Capsule, length=7, radius=5, shapeName=dyn_jacket_bk_A_r_capsule, center=<0 -4 0>}]
 dyn_jacket_bk_B [{shapeType=Capsule, length=7, radius=5, shapeName=dyn_jacket_bk_B_capsule, center=<2 -4 0>}]
 dyn_jacket_bk_B_l [{shapeType=Capsule, length=7, radius=5, shapeName=dyn_jacket_bk_B_l_capsule, center=<2.3794 -4 0>}]
 dyn_jacket_bk_B_r [{shapeType=Capsule, length=7, radius=5, shapeName=dyn_jacket_bk_B_r_capsule, center=<2.3794 -4 0>}]
 dyn_jacket_bk_C [{shapeType=Capsule, length=7, radius=5, shapeName=dyn_jacket_bk_C_capsule, center=<4.2317 -4 0>}]
 dyn_jacket_bk_C_l [{shapeType=Capsule, length=7, radius=5, shapeName=dyn_jacket_bk_C_l_capsule, center=<4.2317 -4 0>}]
 dyn_jacket_bk_C_r [{shapeType=Capsule, length=7, radius=5, shapeName=dyn_jacket_bk_C_r_capsule, center=<4.2317 -4 0>}]
 dyn_jacket_flap_base_l [{shapeType=Capsule, length=0.505, radius=0.505, rotation=<rot 0 -90 0>}]
 dyn_jacket_flap_base_r [{shapeType=Capsule, length=0.505, radius=0.505, rotation=<rot 0 -90 180>}]
 dyn_jacket_flap_l [{shapeType=Capsule, length=3.5, radius=2, center=<3.788 0 0>, rotation=<rot 90 0 0>}]
 dyn_jacket_flap_r [{shapeType=Capsule, length=3.5, radius=2, center=<-3.788 0 0>, rotation=<rot 90 0 180>}]
 dyn_jacket_fr_A_base_l [{shapeType=Capsule, length=1, radius=1}]
 dyn_jacket_fr_A_base_r [{shapeType=Capsule, length=1, radius=1}]
 dyn_jacket_fr_A_l [{shapeType=Capsule, length=10.7876, radius=5, shapeName=dyn_jacket_fr_A_l_capsule, center=<0 -2.5 0>}]
 dyn_jacket_fr_A_r [{shapeType=Capsule, length=10.7876, radius=5, shapeName=dyn_jacket_fr_A_r_capsule, center=<0 -2.5 0>}]
 dyn_jacket_fr_B_l [{shapeType=Capsule, length=12.5395, radius=5, shapeName=dyn_jacket_fr_B_l_capsule, center=<2.3794 -2.5 0>}]
 dyn_jacket_fr_B_r [{shapeType=Capsule, length=12.5395, radius=5, shapeName=dyn_jacket_fr_B_r_capsule, center=<2.3794 -2.5 0>}]
 dyn_jacket_fr_C_l [{
   shapeType=Capsule
   length=12.1799
   radius=5
   shapeName=dyn_jacket_fr_C_l_capsule
   center=<4.2317 -2.5 0>
   rotation=<rot 0.0001 -0.0001 1.8371>
  }]
 dyn_jacket_fr_C_r [{shapeType=Capsule, length=11.2827, radius=5, shapeName=dyn_jacket_fr_C_r_capsule, center=<4.2317 -2.5 0>, rotation=<rot 0 0 -5.1101>}]
 dyn_jacket_side_A_base_l [{shapeType=Capsule, length=1, radius=1}]
 dyn_jacket_side_A_base_r [{shapeType=Capsule, length=1, radius=1}]
 dyn_jacket_side_A_l [{shapeType=Capsule, length=7, radius=5, shapeName=dyn_jacket_side_A_l_capsule, center=<0 -4 0>}]
 dyn_jacket_side_A_r [{shapeType=Capsule, length=7, radius=5, shapeName=dyn_jacket_side_A_r_capsule, center=<0 -4 0>}]
 dyn_jacket_side_B_l [{shapeType=Capsule, length=7, radius=5, shapeName=dyn_jacket_side_B_l_capsule, center=<2.3794 -4 0>}]
 dyn_jacket_side_B_r [{shapeType=Capsule, length=7, radius=5, shapeName=dyn_jacket_side_B_r_capsule, center=<2.3794 -4 0>}]
 dyn_jacket_side_C_l [{shapeType=Capsule, length=7, radius=5, shapeName=dyn_jacket_side_C_l_capsule, center=<4.2317 -4 0>}]
 dyn_jacket_side_C_r [{shapeType=Capsule, length=7, radius=5, shapeName=dyn_jacket_side_C_r_capsule, center=<4.2317 -4 0>}]
 dyn_spray_can [{shapeType=Capsule, length=10.9668, radius=3.4848, center=<4.0084 -0.5648 4.4704>, rotation=<rot 0 72.3756 -76.8791>}]
 dyn_spray_can_base [{shapeType=Capsule, length=0.505, radius=0.505, rotation=<rot 0 -90 0>}]
 pelvis [{shapeType=Capsule, length=11.9601, radius=10.969, center=<-2.1404 -2.0161 0>}]
 spine_01 [{shapeType=Capsule, length=7.7813, radius=9.0102, shapeName=spine_01_capsule, center=<0 -1.2268 0.6193>}]
 spine_02 [{shapeType=Capsule, length=5.7565, radius=8.937, center=<-1.7045 -0.741 0>}]
 spine_03 [{shapeType=Capsule, length=7.0283, radius=8.2922, shapeName=spine_03_capsule, center=<0 0.1906 0>}]
 spine_04 [{shapeType=Capsule, length=6.3502, radius=8.938, shapeName=spine_04_capsule, center=<0 0.5139 0>}]
 thigh_l [{shapeType=Capsule, length=27, radius=7.25, center=<-14.9515 0.0893 0.0976>, rotation=<rot 88.1645 -180 -179.0662>}]
 thigh_r [{shapeType=Capsule, length=27, radius=7.25, center=<14.9515 -0.0893 -0.0976>, rotation=<rot 88.1641 -180 0.9338>}]
}
constraints [
 {bone1Name=dyn_jacket_side_A_l, bone2Name=dyn_jacket_side_A_base_l, swing1LimitDegrees=15, swing2LimitDegrees=15, twistLimitDegrees=0}
 {bone1Name=dyn_jacket_side_B_l, bone2Name=dyn_jacket_side_A_l, swing1LimitDegrees=10, swing2LimitDegrees=15, twistLimitDegrees=0}
 {bone1Name=dyn_jacket_side_C_l, bone2Name=dyn_jacket_side_B_l, swing1LimitDegrees=15, swing2LimitDegrees=15, twistLimitDegrees=0}
 {bone1Name=dyn_jacket_fr_A_l, bone2Name=dyn_jacket_fr_A_base_l, swing1LimitDegrees=10, swing2LimitDegrees=15, twistLimitDegrees=0}
 {bone1Name=dyn_jacket_fr_B_l, bone2Name=dyn_jacket_fr_A_l, swing1LimitDegrees=10, swing2LimitDegrees=15, twistLimitDegrees=0}
 {bone1Name=dyn_jacket_fr_C_l, bone2Name=dyn_jacket_fr_B_l, swing1LimitDegrees=15, swing2LimitDegrees=15, twistLimitDegrees=0}
 {bone1Name=dyn_jacket_bk_A_l, bone2Name=dyn_jacket_bk_A_base_l, swing1LimitDegrees=15, swing2LimitDegrees=5, twistLimitDegrees=0}
 {bone1Name=dyn_jacket_bk_C_l, bone2Name=dyn_jacket_bk_B_l, swing1LimitDegrees=15, swing2LimitDegrees=5, twistLimitDegrees=0}
 {bone1Name=dyn_jacket_bk_B_l, bone2Name=dyn_jacket_bk_A_l, swing1LimitDegrees=15, swing2LimitDegrees=5, twistLimitDegrees=0}
 {bone1Name=dyn_jacket_bk_A, bone2Name=dyn_jacket_bk_A_base, swing1LimitDegrees=15, swing2LimitDegrees=5, twistLimitDegrees=0}
 {bone1Name=dyn_jacket_bk_B, bone2Name=dyn_jacket_bk_A, swing1LimitDegrees=15, swing2LimitDegrees=5, twistLimitDegrees=0}
 {bone1Name=dyn_jacket_bk_C, bone2Name=dyn_jacket_bk_B, swing1LimitDegrees=15, swing2LimitDegrees=5, twistLimitDegrees=0}
 {bone1Name=dyn_jacket_bk_A_r, bone2Name=dyn_jacket_bk_A_base_r, swing1LimitDegrees=15, swing2LimitDegrees=10, twistLimitDegrees=0}
 {bone1Name=dyn_jacket_bk_B_r, bone2Name=dyn_jacket_bk_A_r, swing1LimitDegrees=15, swing2LimitDegrees=10, twistLimitDegrees=0}
 {bone1Name=dyn_jacket_bk_C_r, bone2Name=dyn_jacket_bk_B_r, swing1LimitDegrees=20, swing2LimitDegrees=20, twistLimitDegrees=0}
 {bone1Name=dyn_jacket_side_A_r, bone2Name=dyn_jacket_side_A_base_r, swing1LimitDegrees=20, swing2LimitDegrees=15, twistLimitDegrees=0}
 {bone1Name=dyn_jacket_side_B_r, bone2Name=dyn_jacket_side_A_r, swing1LimitDegrees=20, swing2LimitDegrees=15, twistLimitDegrees=0}
 {bone1Name=dyn_jacket_side_C_r, bone2Name=dyn_jacket_side_B_r, swing1LimitDegrees=30, swing2LimitDegrees=15, twistLimitDegrees=0}
 {bone1Name=dyn_jacket_fr_A_r, bone2Name=dyn_jacket_fr_A_base_r, swing1LimitDegrees=20, swing2LimitDegrees=15, twistLimitDegrees=0}
 {bone1Name=dyn_jacket_fr_B_r, bone2Name=dyn_jacket_fr_A_r, swing1LimitDegrees=20, swing2LimitDegrees=15, twistLimitDegrees=0}
 {bone1Name=dyn_jacket_fr_C_r, bone2Name=dyn_jacket_fr_B_r, swing1LimitDegrees=30, swing2LimitDegrees=15, twistLimitDegrees=0}
 {
  bone1Name=dyn_jacket_side_A_base_l
  bone2Name=spine_04
  swing1LimitDegrees=45
  swing1Motion=Free
  swing2LimitDegrees=45
  swing2Motion=Free
  twistLimitDegrees=45
  twistMotion=Free
 }
 {
  bone1Name=dyn_jacket_fr_A_base_l
  bone2Name=spine_04
  swing1LimitDegrees=45
  swing1Motion=Free
  swing2LimitDegrees=45
  swing2Motion=Free
  twistLimitDegrees=45
  twistMotion=Free
 }
 {
  bone1Name=dyn_jacket_bk_A_base_l
  bone2Name=spine_04
  swing1LimitDegrees=45
  swing1Motion=Free
  swing2LimitDegrees=45
  swing2Motion=Free
  twistLimitDegrees=45
  twistMotion=Free
 }
 {
  bone1Name=dyn_jacket_side_A_base_r
  bone2Name=spine_04
  swing1LimitDegrees=45
  swing1Motion=Free
  swing2LimitDegrees=45
  swing2Motion=Free
  twistLimitDegrees=45
  twistMotion=Free
 }
 {
  bone1Name=dyn_jacket_fr_A_base_r
  bone2Name=spine_04
  swing1LimitDegrees=45
  swing1Motion=Free
  swing2LimitDegrees=45
  swing2Motion=Free
  twistLimitDegrees=45
  twistMotion=Free
 }
 {
  bone1Name=dyn_jacket_bk_A_base_r
  bone2Name=spine_04
  swing1LimitDegrees=45
  swing1Motion=Free
  swing2LimitDegrees=45
  swing2Motion=Free
  twistLimitDegrees=45
  twistMotion=Free
 }
 {
  bone1Name=dyn_jacket_bk_A_base
  bone2Name=spine_04
  swing1LimitDegrees=45
  swing1Motion=Free
  swing2LimitDegrees=45
  swing2Motion=Free
  twistLimitDegrees=45
  twistMotion=Free
 }
 {
  bone1Name=spine_02
  bone2Name=spine_01
  swing1LimitDegrees=45
  swing1Motion=Free
  swing2LimitDegrees=45
  swing2Motion=Free
  twistLimitDegrees=45
  twistMotion=Free
 }
 {
  bone1Name=thigh_l
  bone2Name=pelvis
  swing1LimitDegrees=45
  swing1Motion=Free
  swing2LimitDegrees=45
  swing2Motion=Free
  twistLimitDegrees=45
  twistMotion=Free
 }
 {
  bone1Name=dyn_jacket_flap_base_l
  bone2Name=spine_04
  swing1LimitDegrees=45
  swing1Motion=Free
  swing2LimitDegrees=45
  swing2Motion=Free
  twistLimitDegrees=45
  twistMotion=Free
 }
 {bone1Name=dyn_jacket_flap_l, bone2Name=dyn_jacket_flap_base_l, swing1LimitDegrees=25, swing2LimitDegrees=25, twistLimitDegrees=5}
 {
  bone1Name=dyn_jacket_flap_base_r
  bone2Name=spine_04
  swing1LimitDegrees=45
  swing1Motion=Free
  swing2LimitDegrees=45
  swing2Motion=Free
  twistLimitDegrees=45
  twistMotion=Free
 }
 {bone1Name=dyn_jacket_flap_r, bone2Name=dyn_jacket_flap_base_r, swing1LimitDegrees=25, swing2LimitDegrees=25, twistLimitDegrees=5}
 {
  bone1Name=dyn_spray_can_base
  bone2Name=thigh_l
  swing1LimitDegrees=45
  swing1Motion=Free
  swing2LimitDegrees=45
  swing2Motion=Free
  twistLimitDegrees=45
  twistMotion=Free
 }
 {bone1Name=dyn_spray_can, bone2Name=dyn_spray_can_base, swing1LimitDegrees=25, swing2LimitDegrees=25, twistLimitDegrees=24}
 {
  bone1Name=thigh_r
  bone2Name=pelvis
  swing1LimitDegrees=45
  swing1Motion=Free
  swing2LimitDegrees=45
  swing2Motion=Free
  twistLimitDegrees=45
  twistMotion=Free
 }
 {
  bone1Name=spine_01
  bone2Name=pelvis
  swing1LimitDegrees=45
  swing1Motion=Free
  swing2LimitDegrees=45
  swing2Motion=Free
  twistLimitDegrees=45
  twistMotion=Free
 }
 {
  bone1Name=spine_03
  bone2Name=spine_02
  swing1LimitDegrees=45
  swing1Motion=Free
  swing2LimitDegrees=45
  swing2Motion=Free
  twistLimitDegrees=45
  twistMotion=Free
 }
 {
  bone1Name=spine_04
  bone2Name=spine_03
  swing1LimitDegrees=45
  swing1Motion=Free
  swing2LimitDegrees=45
  swing2Motion=Free
  twistLimitDegrees=45
  twistMotion=Free
 }
 {
  bone1Name=dyn_jacket_side_C_r
  bone2Name=dyn_jacket_fr_C_r
  swing1LimitDegrees=45
  swing1Motion=Locked
  swing2LimitDegrees=45
  swing2Motion=Locked
  twistLimitDegrees=45
  twistMotion=Locked
 }
 {
  bone1Name=dyn_jacket_bk_C_r
  bone2Name=dyn_jacket_side_C_r
  swing1LimitDegrees=45
  swing1Motion=Locked
  swing2LimitDegrees=45
  swing2Motion=Locked
  twistLimitDegrees=45
  twistMotion=Locked
 }
 {
  bone1Name=dyn_jacket_bk_C
  bone2Name=dyn_jacket_bk_C_r
  swing1LimitDegrees=45
  swing1Motion=Locked
  swing2LimitDegrees=45
  swing2Motion=Locked
  twistLimitDegrees=45
  twistMotion=Locked
 }
 {
  bone1Name=dyn_jacket_bk_C_l
  bone2Name=dyn_jacket_side_C_l
  swing1LimitDegrees=45
  swing1Motion=Locked
  swing2LimitDegrees=45
  swing2Motion=Locked
  twistLimitDegrees=45
 }
 {
  bone1Name=dyn_jacket_bk_C
  bone2Name=dyn_jacket_bk_C_l
  swing1LimitDegrees=45
  swing1Motion=Locked
  swing2LimitDegrees=45
  swing2Motion=Locked
  twistLimitDegrees=45
  twistMotion=Locked
 }
 {
  bone1Name=dyn_jacket_side_C_l
  bone2Name=dyn_jacket_fr_C_l
  swing1LimitDegrees=45
  swing1Motion=Locked
  swing2LimitDegrees=45
  swing2Motion=Locked
  twistLimitDegrees=45
 }
]
