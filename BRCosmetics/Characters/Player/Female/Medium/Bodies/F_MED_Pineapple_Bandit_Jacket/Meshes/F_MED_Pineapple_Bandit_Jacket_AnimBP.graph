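# Post-copy AnimBlueprint for the female Pineapple Bandit jacket cosmetic, animating jacket, sleeve flap, collar and spray can bones over the attached parent's pose.
# Weighted multi-bone constraints build a cached base pose.
# The dynamic pose adds corrective offsets, driven by body-bone rotations and crouch-gated left-leg lift, before three RigidBody simulations.
# Output falls back to the base pose when rigid bodies are disabled or LOD exceeds 1.
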AnimBlueprint /BRCosmetics/Characters/Player/Female/Medium/Bodies/F_MED_Pineapple_Bandit_Jacket/Meshes/F_MED_Pineapple_Bandit_Jacket_AnimBP
target_skeleton=&/F_MED_Pineapple_Bandit_Jacket_Skeleton
graphs [
 AnimGraph {nodes [
   n1 = "Transform (Modify) Bone - Bone: dyn_jacket_fr_A_base_l"(Alpha=1, Translation="0,-2,0", Rotation="0,-5,0", Scale="1,1,1") {
    rotationMode=BMM_Additive
    rotationSpace=BCS_BoneSpace
    scaleMode=BMM_Additive
    scaleSpace=BCS_BoneSpace
    translationMode=BMM_Additive
    translationSpace=BCS_BoneSpace
    alphaBoolBlend {blendOption=Linear}
    alphaScaleBiasClamp {inRange {max=100, min=40}}
    boneToModify {boneName=dyn_jacket_fr_A_base_l}
    rotation {yaw=-5}
    translation {y=-2}
   }
   n2 = "Transform (Modify) Bone - Bone: dyn_jacket_side_A_base_l"(ComponentPose=@n1, Alpha=1, Translation="0,-2,0", Rotation="0,-5,0", Scale="1,1,1") {
    rotationMode=BMM_Additive
    rotationSpace=BCS_BoneSpace
    scaleMode=BMM_Additive
    scaleSpace=BCS_BoneSpace
    translationMode=BMM_Additive
    translationSpace=BCS_BoneSpace
    alphaBoolBlend {blendOption=Linear}
    alphaScaleBiasClamp {inRange {max=100, min=40}}
    boneToModify {boneName=dyn_jacket_side_A_base_l}
    rotation {yaw=-5}
    translation {y=-2}
   }
   n3 = "Transform (Modify) Bone - Bone: dyn_jacket_bk_A_base_l"(ComponentPose=@n2, Alpha=1, Translation="0,-2,-5", Rotation="0,0,0", Scale="1,1,1") {
    rotationMode=BMM_Additive
    rotationSpace=BCS_BoneSpace
    scaleMode=BMM_Additive
    scaleSpace=BCS_BoneSpace
    translationMode=BMM_Additive
    translationSpace=BCS_BoneSpace
    alphaBoolBlend {blendOption=Linear}
    alphaScaleBiasClamp {inRange {max=100, min=40}}
    boneToModify {boneName=dyn_jacket_bk_A_base_l}
    translation {y=-2, z=-5}
   }
   n4 = "Transform (Modify) Bone - Bone: dyn_jacket_fr_A_base_r"(ComponentPose=@n3, Alpha=1, Translation="0,2,0", Rotation="0,-5,0", Scale="1,1,1") {
    rotationMode=BMM_Additive
    rotationSpace=BCS_BoneSpace
    scaleMode=BMM_Additive
    scaleSpace=BCS_BoneSpace
    translationMode=BMM_Additive
    translationSpace=BCS_BoneSpace
    alphaBoolBlend {blendOption=Linear}
    alphaScaleBiasClamp {inRange {max=100, min=40}}
    boneToModify {boneName=dyn_jacket_fr_A_base_r}
    rotation {yaw=-5}
    translation {y=2}
   }
   n5 = "Transform (Modify) Bone - Bone: dyn_jacket_side_A_base_r"(ComponentPose=@n4, Alpha=1, Translation="0,2,0", Rotation="0,-5,0", Scale="1,1,1") {
    rotationMode=BMM_Additive
    rotationSpace=BCS_BoneSpace
    scaleMode=BMM_Additive
    scaleSpace=BCS_BoneSpace
    translationMode=BMM_Additive
    translationSpace=BCS_BoneSpace
    alphaBoolBlend {blendOption=Linear}
    alphaScaleBiasClamp {inRange {max=100, min=40}}
    boneToModify {boneName=dyn_jacket_side_A_base_r}
    rotation {yaw=-5}
    translation {y=2}
   }
   n6 = "Transform (Modify) Bone - Bone: dyn_jacket_bk_A_base_r"(ComponentPose=@n5, Alpha=1, Translation="0,2,0", Rotation="0,-5,0", Scale="1,1,1") {
    rotationMode=BMM_Additive
    rotationSpace=BCS_BoneSpace
    scaleMode=BMM_Additive
    scaleSpace=BCS_BoneSpace
    translationMode=BMM_Additive
    translationSpace=BCS_BoneSpace
    alphaBoolBlend {blendOption=Linear}
    alphaScaleBiasClamp {inRange {max=100, min=40}}
    boneToModify {boneName=dyn_jacket_bk_A_base_r}
    rotation {yaw=-5}
    translation {y=2}
   }
   "Transform (Modify) Bone - Bone: dyn_jacket_bk_A_base"(ComponentPose=@n6, Alpha=1, Translation="0,-2,0", Rotation="0,0,0", Scale="1,1,1") {
    rotationMode=BMM_Additive
    rotationSpace=BCS_BoneSpace
    scaleMode=BMM_Additive
    scaleSpace=BCS_BoneSpace
    translationMode=BMM_Additive
    translationSpace=BCS_BoneSpace
    alphaBoolBlend {blendOption=Linear}
    alphaScaleBiasClamp {inRange {max=100, min=40}}
    boneToModify {boneName=dyn_jacket_bk_A_base}
    translation {y=-2}
   }
   Output Pose(Result=Blend Poses by bool(
     BlendPose_0=Use cached pose 'common_pose'()
     BlendPose_1=Use cached pose 'dynamic_pose'()
     bActiveValue=Get isRigidBodyDisabled() -> (isRigidBodyDisabled:bool)
    ) {blendType=Linear}) {layerGroup=None}
   n7 = "Constraint - dyn_jacket_fr_A_base_l (PPPPP)"!(
    ComponentPose=Local To Component(LocalPose=Copy Pose From Mesh() {bUseAttachedParent=true})
    Alpha=1
    ConstraintWeights_0=0.0172
    ConstraintWeights_1:float=0.0636
    ConstraintWeights_2:float=0.8886
    ConstraintWeights_3:float=0.0195
    ConstraintWeights_4:float=0.0111
   ) -> (Pose) {
    alphaBoolBlend {blendOption=Linear}
    boneToModify {boneName=dyn_jacket_fr_A_base_l}
    constraintSetup [
     {targetBone {boneName=spine_02}}
     {targetBone {boneName=spine_03}}
     {targetBone {boneName=spine_04}}
     {targetBone {boneName=spine_05}}
     {targetBone {boneName=clavicle_l}}
    ]
    constraintWeights [1,1,1,1,1]
   }
   n8 = "Constraint - dyn_jacket_fr_B_l (PPPPP)"!(
    ComponentPose=@n7
    Alpha=1
    ConstraintWeights_0=0.1686
    ConstraintWeights_1:float=0.3387
    ConstraintWeights_2:float=0.1231
    ConstraintWeights_3:float=0.2938
    ConstraintWeights_4:float=0.0759
   ) -> (Pose) {
    alphaBoolBlend {blendOption=Linear}
    boneToModify {boneName=dyn_jacket_fr_B_l}
    constraintSetup [
     {targetBone {boneName=spine_01}}
     {targetBone {boneName=spine_02}}
     {targetBone {boneName=spine_03}}
     {targetBone {boneName=spine_04}}
     {targetBone {boneName=pelvis}}
    ]
    constraintWeights [1,1,1,1,1]
   }
   n9 = "Constraint - dyn_jacket_fr_C_l (PPPPP)"!(
    ComponentPose=@n8
    Alpha=1
    ConstraintWeights_0=0.0597
    ConstraintWeights_1:float=0.0998
    ConstraintWeights_2:float=0.0178
    ConstraintWeights_3:float=0.0253
    ConstraintWeights_4:float=0.7974
   ) -> (Pose) {
    alphaBoolBlend {blendOption=Linear}
    boneToModify {boneName=dyn_jacket_fr_C_l}
    constraintSetup [
     {targetBone {boneName=spine_01}}
     {targetBone {boneName=spine_02}}
     {targetBone {boneName=thigh_twist_01_l}}
     {targetBone {boneName=spine_04}}
     {targetBone {boneName=pelvis}}
    ]
    constraintWeights [1,1,1,1,1]
   }
   n10 = "Constraint - dyn_jacket_side_A_base_l (PPPPPP)"!(
    ComponentPose=@n9
    Alpha=1
    ConstraintWeights_0=0.0577
    ConstraintWeights_1:float=0.0393
    ConstraintWeights_2:float=0.1388
    ConstraintWeights_3:float=0.676
    ConstraintWeights_4:float=0.0578
    ConstraintWeights_5:float=0.0303
   ) -> (Pose) {
    alphaBoolBlend {blendOption=Linear}
    boneToModify {boneName=dyn_jacket_side_A_base_l}
    constraintSetup [
     {targetBone {boneName=clavicle_l}}
     {targetBone {boneName=spine_02}}
     {targetBone {boneName=spine_03}}
     {targetBone {boneName=spine_04}}
     {targetBone {boneName=spine_05}}
     {targetBone {boneName=upperarm_twist_01_l}}
    ]
    constraintWeights [1,1,1,1,1,1]
   }
   n11 = "Constraint - dyn_jacket_side_B_l (PPPPP)"!(
    ComponentPose=@n10
    Alpha=1
    ConstraintWeights_0=0.1029
    ConstraintWeights_1:float=0.1183
    ConstraintWeights_2:float=0.118
    ConstraintWeights_3:float=0.2589
    ConstraintWeights_4:float=0.4018
   ) -> (Pose) {
    alphaBoolBlend {blendOption=Linear}
    boneToModify {boneName=dyn_jacket_side_B_l}
    constraintSetup [
     {targetBone {boneName=spine_01}}
     {targetBone {boneName=spine_02}}
     {targetBone {boneName=spine_03}}
     {targetBone {boneName=spine_04}}
     {targetBone {boneName=pelvis}}
    ]
    constraintWeights [1,1,1,1,1]
   }
   n12 = "Constraint - dyn_jacket_side_C_l (PPPPPP)"!(
    ComponentPose=@n11
    Alpha=1
    ConstraintWeights_0=0.0399
    ConstraintWeights_1:float=0.1053
    ConstraintWeights_2:float=0.0474
    ConstraintWeights_3:float=0.0321
    ConstraintWeights_4:float=0.0154
    ConstraintWeights_5:float=0.7599
   ) -> (Pose) {
    alphaBoolBlend {blendOption=Linear}
    boneToModify {boneName=dyn_jacket_side_C_l}
    constraintSetup [
     {targetBone {boneName=thigh_volume_l}}
     {targetBone {boneName=spine_01}}
     {targetBone {boneName=spine_02}}
     {targetBone {boneName=thigh_twist_01_l}}
     {targetBone {boneName=spine_04}}
     {targetBone {boneName=pelvis}}
    ]
    constraintWeights [1,1,1,1,1,1]
   }
   n13 = "Constraint - dyn_jacket_bk_A_base_l (PPPPPPPP)"!(
    ComponentPose=@n12
    Alpha=1
    ConstraintWeights_0=0.1829
    ConstraintWeights_1:float=0.039
    ConstraintWeights_2:float=0.0903
    ConstraintWeights_3:float=0.1244
    ConstraintWeights_4:float=0.4064
    ConstraintWeights_5:float=0.1236
    ConstraintWeights_6:float=0.0215
    ConstraintWeights_7:float=0.0119
   ) -> (Pose) {
    alphaBoolBlend {blendOption=Linear}
    boneToModify {boneName=dyn_jacket_bk_A_base_l}
    constraintSetup [
     {targetBone {boneName=clavicle_l}}
     {targetBone {boneName=spine_01}}
     {targetBone {boneName=spine_02}}
     {targetBone {boneName=spine_03}}
     {targetBone {boneName=spine_04}}
     {targetBone {boneName=spine_05}}
     {targetBone {boneName=upperarm_twist_01_l}}
     {targetBone {boneName=pelvis}}
    ]
    constraintWeights [1,1,1,1,1,1,1,1]
   }
   n14 = "Constraint - dyn_jacket_bk_B_l (PPPPPPP)"!(
    ComponentPose=@n13
    Alpha=1
    ConstraintWeights_0=0.0314
    ConstraintWeights_1:float=0.1498
    ConstraintWeights_2:float=0.2279
    ConstraintWeights_3:float=0.1515
    ConstraintWeights_4:float=0.212
    ConstraintWeights_5:float=0.0234
    ConstraintWeights_6:float=0.204
   ) -> (Pose) {
    alphaBoolBlend {blendOption=Linear}
    boneToModify {boneName=dyn_jacket_bk_B_l}
    constraintSetup [
     {targetBone {boneName=clavicle_l}}
     {targetBone {boneName=spine_01}}
     {targetBone {boneName=spine_02}}
     {targetBone {boneName=spine_03}}
     {targetBone {boneName=spine_04}}
     {targetBone {boneName=spine_05}}
     {targetBone {boneName=pelvis}}
    ]
    constraintWeights [1,1,1,1,1,1,1]
   }
   n15 = "Constraint - dyn_jacket_bk_C_l (PPPPPP)"!(
    ComponentPose=@n14
    Alpha=1
    ConstraintWeights_0=0.0683
    ConstraintWeights_1:float=0.1312
    ConstraintWeights_2:float=0.073
    ConstraintWeights_3:float=0.0108
    ConstraintWeights_4:float=0.7037
    ConstraintWeights_5:float=0.0129
   ) -> (Pose) {
    alphaBoolBlend {blendOption=Linear}
    boneToModify {boneName=dyn_jacket_bk_C_l}
    constraintSetup [
     {targetBone {boneName=thigh_volume_l}}
     {targetBone {boneName=spine_01}}
     {targetBone {boneName=spine_02}}
     {targetBone {boneName=spine_03}}
     {targetBone {boneName=pelvis}}
     {targetBone {boneName=thigh_twist_01_l}}
    ]
    constraintWeights [1,1,1,1,1,1]
   }
   n16 = "Constraint - dyn_jacket_fr_A_base_r (PPPPPP)"!(
    ComponentPose=@n15
    Alpha=1
    ConstraintWeights_0=0.0002
    ConstraintWeights_1:float=0.0208
    ConstraintWeights_2:float=0.0733
    ConstraintWeights_3:float=0.8796
    ConstraintWeights_4:float=0.0167
    ConstraintWeights_5:float=0.0094
   ) -> (Pose) {
    alphaBoolBlend {blendOption=Linear}
    boneToModify {boneName=dyn_jacket_fr_A_base_r}
    constraintSetup [
     {targetBone {boneName=spine_01}}
     {targetBone {boneName=spine_02}}
     {targetBone {boneName=spine_03}}
     {targetBone {boneName=spine_04}}
     {targetBone {boneName=spine_05}}
     {targetBone {boneName=clavicle_r}}
    ]
    constraintWeights [1,1,1,1,1,1]
   }
   n17 = "Constraint - dyn_jacket_fr_B_r (PPPPP)"!(
    ComponentPose=@n16
    Alpha=1
    ConstraintWeights_0=0.126
    ConstraintWeights_1:float=0.2433
    ConstraintWeights_2:float=0.1034
    ConstraintWeights_3:float=0.284
    ConstraintWeights_4:float=0.2434
   ) -> (Pose) {
    alphaBoolBlend {blendOption=Linear}
    boneToModify {boneName=dyn_jacket_fr_B_r}
    constraintSetup [
     {targetBone {boneName=spine_01}}
     {targetBone {boneName=spine_02}}
     {targetBone {boneName=spine_03}}
     {targetBone {boneName=spine_04}}
     {targetBone {boneName=pelvis}}
    ]
    constraintWeights [1,1,1,1,1]
   }
   n18 = "Constraint - dyn_jacket_fr_C_r (PPPPP)"!(
    ComponentPose=@n17
    Alpha=1
    ConstraintWeights_0=0.0563
    ConstraintWeights_1:float=0.0941
    ConstraintWeights_2:float=0.0235
    ConstraintWeights_3:float=0.0219
    ConstraintWeights_4:float=0.8042
   ) -> (Pose) {
    alphaBoolBlend {blendOption=Linear}
    boneToModify {boneName=dyn_jacket_fr_C_r}
    constraintSetup [
     {targetBone {boneName=spine_01}}
     {targetBone {boneName=spine_02}}
     {targetBone {boneName=spine_04}}
     {targetBone {boneName=thigh_twist_01_r}}
     {targetBone {boneName=pelvis}}
    ]
    constraintWeights [1,1,1,1,1]
   }
   n19 = "Constraint - dyn_jacket_side_A_base_r (PPPPPPPP)"!(
    ComponentPose=@n18
    Alpha=1
    ConstraintWeights_0=0.1074
    ConstraintWeights_1:float=0.0138
    ConstraintWeights_2:float=0.0415
    ConstraintWeights_3:float=0.1087
    ConstraintWeights_4:float=0.5993
    ConstraintWeights_5:float=0.0546
    ConstraintWeights_6:float=0.0739
    ConstraintWeights_7:float=0.0007
   ) -> (Pose) {
    alphaBoolBlend {blendOption=Linear}
    boneToModify {boneName=dyn_jacket_side_A_base_r}
    constraintSetup [
     {targetBone {boneName=upperarm_twist_01_r}}
     {targetBone {boneName=spine_01}}
     {targetBone {boneName=spine_02}}
     {targetBone {boneName=spine_03}}
     {targetBone {boneName=spine_04}}
     {targetBone {boneName=spine_05}}
     {targetBone {boneName=clavicle_r}}
     {targetBone {boneName=pelvis}}
    ]
    constraintWeights [1,1,1,1,1,1,1,1]
   }
   n20 = "Constraint - dyn_jacket_side_B_r (PPPPP)"!(
    ComponentPose=@n19
    Alpha=1
    ConstraintWeights_0=0.1023
    ConstraintWeights_1:float=0.1179
    ConstraintWeights_2:float=0.1181
    ConstraintWeights_3:float=0.2604
    ConstraintWeights_4:float=0.4013
   ) -> (Pose) {
    alphaBoolBlend {blendOption=Linear}
    boneToModify {boneName=dyn_jacket_side_B_r}
    constraintSetup [
     {targetBone {boneName=spine_01}}
     {targetBone {boneName=spine_02}}
     {targetBone {boneName=spine_03}}
     {targetBone {boneName=spine_04}}
     {targetBone {boneName=pelvis}}
    ]
    constraintWeights [1,1,1,1,1]
   }
   n21 = "Constraint - dyn_jacket_side_C_r (PPPPPP)"!(
    ComponentPose=@n20
    Alpha=1
    ConstraintWeights_0=0.0333
    ConstraintWeights_1:float=0.104
    ConstraintWeights_2:float=0.0466
    ConstraintWeights_3:float=0.0403
    ConstraintWeights_4:float=0.7611
    ConstraintWeights_5:float=0.0146
   ) -> (Pose) {
    alphaBoolBlend {blendOption=Linear}
    boneToModify {boneName=dyn_jacket_side_C_r}
    constraintSetup [
     {targetBone {boneName=thigh_twist_01_r}}
     {targetBone {boneName=spine_01}}
     {targetBone {boneName=spine_02}}
     {targetBone {boneName=thigh_volume_r}}
     {targetBone {boneName=pelvis}}
     {targetBone {boneName=spine_04}}
    ]
    constraintWeights [1,1,1,1,1,1]
   }
   n22 = "Constraint - dyn_jacket_bk_A_base_r (PPPPPPPPP)"!(
    ComponentPose=@n21
    Alpha=1
    ConstraintWeights_0=0
    ConstraintWeights_1:float=0.0204
    ConstraintWeights_2:float=0.0386
    ConstraintWeights_3:float=0.0897
    ConstraintWeights_4:float=0.1235
    ConstraintWeights_5:float=0.4042
    ConstraintWeights_6:float=0.1267
    ConstraintWeights_7:float=0.1857
    ConstraintWeights_8:float=0.0112
   ) -> (Pose) {
    alphaBoolBlend {blendOption=Linear}
    boneToModify {boneName=dyn_jacket_bk_A_base_r}
    constraintSetup [
     {targetBone {boneName=clavicle_l}}
     {targetBone {boneName=upperarm_twist_01_r}}
     {targetBone {boneName=spine_01}}
     {targetBone {boneName=spine_02}}
     {targetBone {boneName=spine_03}}
     {targetBone {boneName=spine_04}}
     {targetBone {boneName=spine_05}}
     {targetBone {boneName=clavicle_r}}
     {targetBone {boneName=pelvis}}
    ]
    constraintWeights [1,1,1,1,1,1,1,1,1]
   }
   n23 = "Constraint - dyn_jacket_bk_B_r (PPPPPPP)"!(
    ComponentPose=@n22
    Alpha=1
    ConstraintWeights_0=0.1465
    ConstraintWeights_1:float=0.2231
    ConstraintWeights_2:float=0.1514
    ConstraintWeights_3:float=0.2192
    ConstraintWeights_4:float=0.0241
    ConstraintWeights_5:float=0.0326
    ConstraintWeights_6:float=0.2031
   ) -> (Pose) {
    alphaBoolBlend {blendOption=Linear}
    boneToModify {boneName=dyn_jacket_bk_B_r}
    constraintSetup [
     {targetBone {boneName=spine_01}}
     {targetBone {boneName=spine_02}}
     {targetBone {boneName=spine_03}}
     {targetBone {boneName=spine_04}}
     {targetBone {boneName=spine_05}}
     {targetBone {boneName=clavicle_r}}
     {targetBone {boneName=pelvis}}
    ]
    constraintWeights [1,1,1,1,1,1,1]
   }
   n24 = "Constraint - dyn_jacket_bk_C_r (PPPPPPP)"!(
    ComponentPose=@n23
    Alpha=1
    ConstraintWeights_0=0.0131
    ConstraintWeights_1:float=0.1312
    ConstraintWeights_2:float=0.0734
    ConstraintWeights_3:float=0.0111
    ConstraintWeights_4:float=0.0011
    ConstraintWeights_5:float=0.0677
    ConstraintWeights_6:float=0.7023
   ) -> (Pose) {
    alphaBoolBlend {blendOption=Linear}
    boneToModify {boneName=dyn_jacket_bk_C_r}
    constraintSetup [
     {targetBone {boneName=thigh_twist_01_r}}
     {targetBone {boneName=spine_01}}
     {targetBone {boneName=spine_02}}
     {targetBone {boneName=spine_03}}
     {targetBone {boneName=spine_04}}
     {targetBone {boneName=thigh_volume_r}}
     {targetBone {boneName=pelvis}}
    ]
    constraintWeights [1,1,1,1,1,1,1]
   }
   n25 = "Constraint - dyn_jacket_bk_A_base (PPPPPPP)"!(
    ComponentPose=@n24
    Alpha=1
    ConstraintWeights_0=0.0615
    ConstraintWeights_1:float=0.0477
    ConstraintWeights_2:float=0.118
    ConstraintWeights_3:float=0.1505
    ConstraintWeights_4:float=0.405
    ConstraintWeights_5:float=0.1574
    ConstraintWeights_6:float=0.06
   ) -> (Pose) {
    alphaBoolBlend {blendOption=Linear}
    boneToModify {boneName=dyn_jacket_bk_A_base}
    constraintSetup [
     {targetBone {boneName=clavicle_l}}
     {targetBone {boneName=spine_01}}
     {targetBone {boneName=spine_02}}
     {targetBone {boneName=spine_03}}
     {targetBone {boneName=spine_04}}
     {targetBone {boneName=spine_05}}
     {targetBone {boneName=clavicle_r}}
    ]
    constraintWeights [1,1,1,1,1,1,1]
   }
   n26 = "Constraint - dyn_jacket_bk_B (PPPPPP)"!(
    ComponentPose=@n25
    Alpha=1
    ConstraintWeights_0=0.1693
    ConstraintWeights_1:float=0.2902
    ConstraintWeights_2:float=0.2021
    ConstraintWeights_3:float=0.168
    ConstraintWeights_4:float=0.0225
    ConstraintWeights_5:float=0.1478
   ) -> (Pose) {
    alphaBoolBlend {blendOption=Linear}
    boneToModify {boneName=dyn_jacket_bk_B}
    constraintSetup [
     {targetBone {boneName=spine_01}}
     {targetBone {boneName=spine_02}}
     {targetBone {boneName=spine_03}}
     {targetBone {boneName=spine_04}}
     {targetBone {boneName=spine_05}}
     {targetBone {boneName=pelvis}}
    ]
    constraintWeights [1,1,1,1,1,1]
   }
   n27 = "Constraint - dyn_jacket_bk_C (PPPPP)"!(
    ComponentPose=@n26
    Alpha=1
    ConstraintWeights_0=0.1151
    ConstraintWeights_1:float=0.0485
    ConstraintWeights_2:float=0.0411
    ConstraintWeights_3:float=0.0437
    ConstraintWeights_4:float=0.7516
   ) -> (Pose) {
    alphaBoolBlend {blendOption=Linear}
    boneToModify {boneName=dyn_jacket_bk_C}
    constraintSetup [
     {targetBone {boneName=spine_01}}
     {targetBone {boneName=spine_02}}
     {targetBone {boneName=thigh_volume_r}}
     {targetBone {boneName=thigh_volume_l}}
     {targetBone {boneName=pelvis}}
    ]
    constraintWeights [1,1,1,1,1]
   }
   n28 = "Constraint - dyn_jacket_flap_base_r (PPPP)"!(
    ComponentPose=@n27
    Alpha=1
    ConstraintWeights_0=0.6021
    ConstraintWeights_1:float=0.2018
    ConstraintWeights_2:float=0.14
    ConstraintWeights_3:float=0.0561
   ) -> (Pose) {
    alphaBoolBlend {blendOption=Linear}
    boneToModify {boneName=dyn_jacket_flap_base_r}
    constraintSetup [
     {targetBone {boneName=lowerarm_twist_02_r}}
     {targetBone {boneName=upperarm_r}}
     {targetBone {boneName=lowerarm_r}}
     {targetBone {boneName=upperarm_twist_02_r}}
    ]
    constraintWeights [1,1,1,1]
   }
   n29 = "Constraint - dyn_jacket_flap_base_l (PPPP)"!(
    ComponentPose=@n28
    Alpha=1
    ConstraintWeights_0=0.14
    ConstraintWeights_1:float=0.0561
    ConstraintWeights_2:float=0.2018
    ConstraintWeights_3:float=0.6021
   ) -> (Pose) {
    alphaBoolBlend {blendOption=Linear}
    boneToModify {boneName=dyn_jacket_flap_base_l}
    constraintSetup [
     {targetBone {boneName=lowerarm_l}}
     {targetBone {boneName=upperarm_twist_02_l}}
     {targetBone {boneName=upperarm_l}}
     {targetBone {boneName=lowerarm_twist_02_l}}
    ]
    constraintWeights [1,1,1,1]
   }
   n30 = "Constraint - dyn_spray_can_base (PPPP)"!(
    ComponentPose=@n29
    Alpha=1
    ConstraintWeights_0=0.3903
    ConstraintWeights_1:float=0.0163
    ConstraintWeights_2:float=0.5298
    ConstraintWeights_3:float=0.0635
   ) -> (Pose) {
    alphaBoolBlend {blendOption=Linear}
    boneToModify {boneName=dyn_spray_can_base}
    constraintSetup [
     {targetBone {boneName=pelvis}}
     {targetBone {boneName=thigh_volume_l}}
     {targetBone {boneName=thigh_twist_01_l}}
     {targetBone {boneName=thigh_l}}
    ]
    constraintWeights [1,1,1,1]
   }
   n31 = "Constraint - dyn_collar_front_base_l (PPP)"!(ComponentPose=@n30, Alpha=1, ConstraintWeights_0=0.0382, ConstraintWeights_1:float=0.0207, ConstraintWeights_2:float=0.9411) -> (Pose) {
    alphaBoolBlend {blendOption=Linear}
    boneToModify {boneName=dyn_collar_front_base_l}
    constraintSetup [{targetBone {boneName=upperarm_twist_01_l}}, {targetBone {boneName=spine_05}}, {targetBone {boneName=clavicle_l}}]
    constraintWeights [1,1,1]
   }
   n32 = "Constraint - dyn_collar_side_base_l (PP)"!(ComponentPose=@n31, Alpha=1, ConstraintWeights_0=0.0969, ConstraintWeights_1:float=0.9031) -> (Pose) {
    alphaBoolBlend {blendOption=Linear}
    boneToModify {boneName=dyn_collar_side_base_l}
    constraintSetup [{targetBone {boneName=spine_05}}, {targetBone {boneName=clavicle_l}}]
    constraintWeights [1,1]
   }
   n33 = "Constraint - dyn_collar_front_base_r (PPP)"!(ComponentPose=@n32, Alpha=1, ConstraintWeights_0=0.0856, ConstraintWeights_1:float=0.8984, ConstraintWeights_2:float=0.016) -> (Pose) {
    alphaBoolBlend {blendOption=Linear}
    boneToModify {boneName=dyn_collar_front_base_r}
    constraintSetup [{targetBone {boneName=upperarm_twist_01_r}}, {targetBone {boneName=clavicle_r}}, {targetBone {boneName=spine_05}}]
    constraintWeights [1,1,1]
   }
   n34 = "Constraint - dyn_collar_side_base_r (PPPP)"!(
    ComponentPose=@n33
    Alpha=1
    ConstraintWeights_0=0.0002
    ConstraintWeights_1:float=0.8654
    ConstraintWeights_2:float=0.134
    ConstraintWeights_3:float=0.0004
   ) -> (Pose) {
    alphaBoolBlend {blendOption=Linear}
    boneToModify {boneName=dyn_collar_side_base_r}
    constraintSetup [
     {targetBone {boneName=upperarm_twist_01_r}}
     {targetBone {boneName=clavicle_r}}
     {targetBone {boneName=spine_05}}
     {targetBone {boneName=clavicle_l}}
    ]
    constraintWeights [1,1,1,1]
   }
   n35 = "Constraint - dyn_collar_back_base (PPP)"!(ComponentPose=@n34, Alpha=1, ConstraintWeights_0=0.1208, ConstraintWeights_1:float=0.7729, ConstraintWeights_2:float=0.1063) -> (Pose) {
    alphaBoolBlend {blendOption=Linear}
    boneToModify {boneName=dyn_collar_back_base}
    constraintSetup [{targetBone {boneName=clavicle_r}}, {targetBone {boneName=spine_05}}, {targetBone {boneName=clavicle_l}}]
    constraintWeights [1,1,1]
   }
   Save cached pose 'common_pose'(Pose=Component To Local(ComponentPose=@n35)) {cacheName=common_pose}
   n36 = Get Spine_left() -> (Spine_left:double)
   n37 = Get jacketLegAlpha_L() -> (jacketLegAlpha_L:double)
   n38 = "Transform (Modify) Bone - Bone: dyn_jacket_fr_A_base_l"(
    ComponentPose=Local To Component(LocalPose=Use cached pose 'common_pose'())
    Alpha=@n37
    Translation="0,0,-2"
    Rotation="-0.600,-17.294,-85.861"
    Scale="1,1,1"
   ) {
    rotationMode=BMM_Additive
    rotationSpace=BCS_BoneSpace
    scaleMode=BMM_Additive
    scaleSpace=BCS_BoneSpace
    translationMode=BMM_Additive
    translationSpace=BCS_BoneSpace
    alphaBoolBlend {blendOption=Linear}
    alphaScaleBiasClamp {inRange {max=100, min=40}}
    boneToModify {boneName=dyn_jacket_fr_A_base_l}
    rotation=<rot -0.6 -17.294 -85.861>
    translation {z=-2}
   }
   n39 = "Transform (Modify) Bone - Bone: dyn_jacket_side_A_base_l"(ComponentPose=@n38, Alpha=@n37, Translation="0,0,0", Rotation="-41.857,-7.724,-11.546", Scale="1,1,1") {
    rotationMode=BMM_Additive
    rotationSpace=BCS_BoneSpace
    scaleMode=BMM_Additive
    scaleSpace=BCS_BoneSpace
    translationMode=BMM_Additive
    translationSpace=BCS_BoneSpace
    alphaBoolBlend {blendOption=Linear}
    alphaScaleBiasClamp {inRange {max=100, min=40}}
    boneToModify {boneName=dyn_jacket_side_A_base_l}
    rotation=<rot -41.857 -7.724 -11.546>
   }
   n40 = "Transform (Modify) Bone - Bone: dyn_jacket_fr_A_base_l"(ComponentPose=@n39, Alpha=Get LegUpAndSpine() -> (LegUpAndSpine:double), Translation="0,0,0", Rotation="5,-30,-15", Scale="1,1,1") {
    rotationMode=BMM_Additive
    rotationSpace=BCS_BoneSpace
    translationMode=BMM_Additive
    translationSpace=BCS_BoneSpace
    alphaBoolBlend {blendOption=Linear}
    boneToModify {boneName=dyn_jacket_fr_A_base_l}
    rotation=<rot 5 -30 -15>
   }
   n41 = "Transform (Modify) Bone - Bone: dyn_jacket_fr_A_base_r"(ComponentPose=@n40, Alpha=Get legUp_r() -> (legUp_r:double), Translation="0,0,0", Rotation="-0.129,0.232,71.611", Scale="1,1,1") {
    rotationMode=BMM_Additive
    rotationSpace=BCS_BoneSpace
    translationMode=BMM_Additive
    translationSpace=BCS_BoneSpace
    alphaBoolBlend {blendOption=Linear}
    boneToModify {boneName=dyn_jacket_fr_A_base_r}
    rotation=<rot -0.129 0.232 71.611>
   }
   n42 = "Transform (Modify) Bone - Bone: dyn_collar_front_base_r"(ComponentPose=@n41, Alpha=Get armFwd_R() -> (armFwd_R:double), Translation="2,0,2", Rotation="50,0,80", Scale="1,1,1") {
    rotationMode=BMM_Additive
    rotationSpace=BCS_BoneSpace
    scaleMode=BMM_Additive
    scaleSpace=BCS_BoneSpace
    translationMode=BMM_Additive
    translationSpace=BCS_BoneSpace
    alphaBoolBlend {blendOption=Linear}
    alphaScaleBiasClamp {bMapRange=true, inRange {max=100, min=40}}
    boneToModify {boneName=dyn_collar_front_base_r}
    rotation {pitch=50, roll=80}
    translation {x=2, z=2}
   }
   n43 = Get headFwdBkRot() -> (headFwdBkRot:double)
   n44 = "Transform (Modify) Bone - Bone: dyn_collar_front_base_l"(ComponentPose=@n42, Alpha=@n43, Translation="0,0,1", Rotation="-5,0,5", Scale="1,1,1") {
    rotationMode=BMM_Additive
    rotationSpace=BCS_BoneSpace
    translationMode=BMM_Additive
    translationSpace=BCS_BoneSpace
    alphaBoolBlend {blendOption=Linear}
    alphaScaleBiasClamp {bMapRange=true, inRange {max=0, min=-50}, outRange {max=0, min=1}}
    boneToModify {boneName=dyn_collar_front_base_l}
    rotation {pitch=-5, roll=5}
    translation {z=1}
   }
   n45 = "Transform (Modify) Bone - Bone: dyn_collar_side_base_l"(ComponentPose=@n44, Alpha=@n43, Translation="0,0,0", Rotation="-60,0,0", Scale="1,1,1") {
    rotationMode=BMM_Additive
    rotationSpace=BCS_BoneSpace
    translationMode=BMM_Additive
    translationSpace=BCS_BoneSpace
    alphaBoolBlend {blendOption=Linear}
    alphaScaleBiasClamp {bMapRange=true, inRange {max=0, min=-50}, outRange {max=0, min=1}}
    boneToModify {boneName=dyn_collar_side_base_l}
    rotation {pitch=-60}
   }
   n46 = "Transform (Modify) Bone - Bone: dyn_collar_back_base"(ComponentPose=@n45, Alpha=@n43, Translation="0,0,0", Rotation="-25,0,0", Scale="1,1,1") {
    rotationMode=BMM_Additive
    rotationSpace=BCS_BoneSpace
    translationMode=BMM_Additive
    translationSpace=BCS_BoneSpace
    alphaBoolBlend {blendOption=Linear}
    alphaScaleBiasClamp {bMapRange=true, inRange {max=0, min=-50}, outRange {max=0, min=1}}
    boneToModify {boneName=dyn_collar_back_base}
    rotation {pitch=-25}
   }
   n47 = "Transform (Modify) Bone - Bone: dyn_collar_side_base_r"(ComponentPose=@n46, Alpha=@n43, Translation="0,0,0", Rotation="-60,0,0", Scale="1,1,1") {
    rotationMode=BMM_Additive
    rotationSpace=BCS_BoneSpace
    translationMode=BMM_Additive
    translationSpace=BCS_BoneSpace
    alphaBoolBlend {blendOption=Linear}
    alphaScaleBiasClamp {bMapRange=true, inRange {max=0, min=-50}, outRange {max=0, min=1}}
    boneToModify {boneName=dyn_collar_side_base_r}
    rotation {pitch=-60}
   }
   n48 = Get Spine_Right() -> (Spine_Right:double)
   n49 = "Transform (Modify) Bone - Bone: dyn_jacket_side_A_base_r"(ComponentPose=@n47, Alpha=@n48, Translation="-0.139,-4.017,0.105", Rotation="0.085,-38.462,-0.082", Scale="1,1,1") {
    rotationMode=BMM_Additive
    rotationSpace=BCS_BoneSpace
    translationMode=BMM_Additive
    translationSpace=BCS_BoneSpace
    alphaBoolBlend {blendOption=Linear}
    alphaScaleBiasClamp {bInterpResult=true, interpSpeedDecreasing=25, interpSpeedIncreasing=25}
    boneToModify {boneName=dyn_jacket_side_A_base_r}
    rotation=<rot 0.085 -38.462 -0.082>
    translation=<-0.139 -4.017 0.105>
   }
   n50 = "Transform (Modify) Bone - Bone: dyn_jacket_fr_A_base_r"(ComponentPose=@n49, Alpha=@n48, Translation="-0.000,-3.330,-0.000", Rotation="0,-35,0", Scale="1,1,1") {
    rotationMode=BMM_Additive
    rotationSpace=BCS_BoneSpace
    translationMode=BMM_Additive
    translationSpace=BCS_BoneSpace
    alphaBoolBlend {blendOption=Linear}
    alphaScaleBiasClamp {bInterpResult=true, interpSpeedDecreasing=15, interpSpeedIncreasing=15}
    boneToModify {boneName=dyn_jacket_fr_A_base_r}
    rotation {yaw=-35}
    translation {y=-3.33}
   }
   n51 = "Transform (Modify) Bone - Bone: dyn_jacket_bk_A_base_r"(ComponentPose=@n50, Alpha=@n48, Translation="0.945,-1.817,-1.145", Rotation="-0.244,-42.724,-0.294", Scale="1,1,1") {
    rotationMode=BMM_Additive
    rotationSpace=BCS_BoneSpace
    translationMode=BMM_Additive
    translationSpace=BCS_BoneSpace
    alphaBoolBlend {blendOption=Linear}
    alphaScaleBiasClamp {bInterpResult=true, interpSpeedDecreasing=20, interpSpeedIncreasing=20}
    boneToModify {boneName=dyn_jacket_bk_A_base_r}
    rotation=<rot -0.244 -42.724 -0.294>
    translation=<0.945 -1.8168 -1.1448>
   }
   n52 = "Transform (Modify) Bone - Bone: dyn_jacket_side_A_base_l"(ComponentPose=@n51, Alpha=@n36, Translation="0,-2,0", Rotation="0,-30,0", Scale="1,1,1") {
    rotationMode=BMM_Additive
    rotationSpace=BCS_BoneSpace
    translationMode=BMM_Additive
    translationSpace=BCS_BoneSpace
    alphaBoolBlend {blendOption=Linear}
    alphaScaleBiasClamp {bInterpResult=true, interpSpeedDecreasing=15, interpSpeedIncreasing=15}
    boneToModify {boneName=dyn_jacket_side_A_base_l}
    rotation {yaw=-30}
    translation {y=-2}
   }
   n53 = "Transform (Modify) Bone - Bone: dyn_jacket_fr_A_base_l"(ComponentPose=@n52, Alpha=@n36, Translation="0,-2,0", Rotation="0,-30,0", Scale="1,1,1") {
    rotationMode=BMM_Additive
    rotationSpace=BCS_BoneSpace
    translationMode=BMM_Additive
    translationSpace=BCS_BoneSpace
    alphaBoolBlend {blendOption=Linear}
    alphaScaleBiasClamp {bInterpResult=true, interpSpeedDecreasing=15, interpSpeedIncreasing=15}
    boneToModify {boneName=dyn_jacket_fr_A_base_l}
    rotation {yaw=-30}
    translation {y=-2}
   }
   n54 = RigidBody(ComponentPose=@n53) {
    bClampLinearTranslationLimitToRefPose=true
    bOverrideWorldGravity=true
    overridePhysicsAsset=/BRCosmetics/Characters/Player/Female/Medium/Bodies/F_MED_Pineapple_Bandit/Meshes/F_MED_Pineapple_Bandit_Physics
    simulationSpace=BaseBoneSpace
    alphaBoolBlend {blendOption=Linear}
    baseBoneRef {boneName=root}
    componentLinearAccScale=<0.4 0.4 0.4>
    componentLinearVelScale=<0.4 0.4 0.4>
    overrideWorldGravity {z=-980}
   }
   n55 = RigidBody(ComponentPose=@n54) {
    bClampLinearTranslationLimitToRefPose=true
    bOverrideWorldGravity=true
    overridePhysicsAsset=&/F_MED_Pineapple_Bandit_Jacket_Physics
    simulationSpace=BaseBoneSpace
    alphaBoolBlend {blendOption=Linear}
    baseBoneRef {boneName=pelvis}
    componentLinearAccScale=<0.25 0.25 0.25>
    componentLinearVelScale=<0.25 0.25 0.25>
    overrideWorldGravity {z=-980}
   }
   n56 = RigidBody(ComponentPose=@n55) {
    bClampLinearTranslationLimitToRefPose=true
    bOverrideWorldGravity=true
    overridePhysicsAsset=&/F_MED_Pineapple_Bandit_Jacket_Collar_Physics
    simulationSpace=BaseBoneSpace
    alphaBoolBlend {blendOption=Linear}
    baseBoneRef {boneName=spine_02}
    componentLinearAccScale=<0.3 0.3 0.3>
    componentLinearVelScale=<0.3 0.3 0.3>
    overrideWorldGravity {z=-980}
   }
   Save cached pose 'dynamic_pose'(Pose=Component To Local(ComponentPose=@n56)) {cacheName=dynamic_pose}
  ]}
 EventGraph {nodes [
   Event BlueprintUpdateAnimation()
   > n1 = Sequence!(execute) -> (then_0, then_1, then_2:Exec)
   n2 = Sequence!(execute=@n1.then_1) -> (then_0, then_1, then_2, then_3:Exec, then_4:Exec, then_5:Exec)
   n3 = GetOwnerSkeletalMeshComponent()
   n4 = GetDeltaTransformFromRefPose(self=@n3, BoneName=thigh_l, BaseName=pelvis)
   Set legUp_L(execute=@n2.then_1, legUp_L:double=BreakRotator(InRot=BreakTransform(InTransform=@n4).Rotation).Yaw) -> (Output_Get:double)
   n5 = BreakRotator(InRot=BreakTransform(InTransform=GetDeltaTransformFromRefPose(self=@n3, BoneName=spine_05, BaseName=pelvis)).Rotation)
   Set LegUpAndSpine(
    execute=@n2.then_2
    LegUpAndSpine:double=float * float!(
     A=Get jacketLegAlpha_L()
     B=MapRangeClamped(Value=@n5, InRangeA=-35, InRangeB=-10, OutRangeA=1, OutRangeB=0)
     C:double=MapRangeClamped(Value=@n5.Yaw, InRangeA=30, InRangeB=10, OutRangeA=1, OutRangeB=0)
    ) -> (ReturnValue) {is=title K2Node_PromotableOperator}
   ) -> (Output_Get:double)
   n6 = Get PelvisRoot() -> (PelvisRoot:double)
   > Set Spine_Right(Spine_Right:double=float * float!(A=MapRangeClamped(Value=@n5.Pitch, InRangeA=13, InRangeB=5, OutRangeA=1, OutRangeB=0), B=@n6) -> (ReturnValue)) -> (Output_Get:double)
   > Set Spine_left(Spine_left:double=float * float(A=MapRangeClamped(Value=@n5.Pitch, InRangeA=-10, InRangeB=-5, OutRangeA=1, OutRangeB=0), B=@n6)) -> (Output_Get:double)
   n7 = GetDeltaTransformFromRefPose(self=@n3, BoneName=upperarm_r, BaseName=clavicle_r)
   Set armFwd_R(execute=@n2.then_0, armFwd_R:double=BreakRotator(InRot=BreakTransform(InTransform=@n7).Rotation).Yaw) -> (Output_Get:double)
   n8 = GetDeltaTransformFromRefPose(self=@n3, BoneName=head, BaseName=spine_05)
   Set headFwdBkRot(execute=@n2.then_3, headFwdBkRot:double=BreakRotator(InRot=BreakTransform(InTransform=@n8).Rotation).Yaw) -> (Output_Get:double)
   n9 = BreakTransform(InTransform=GetDeltaTransformFromRefPose(self=@n3, BoneName=thigh_r, BaseName=pelvis))
   Set legUp_r(
    execute=@n2.then_4
    legUp_r:double=MapRangeClamped(Value=BreakRotator(InRot=@n9.Rotation).Yaw, InRangeA=-80, InRangeB=-20, OutRangeA=1, OutRangeB=0)
   ) -> (Output_Get:double)
   n10 = BreakTransform(InTransform=GetDeltaTransformFromRefPose(self=@n3, BoneName=pelvis, BaseName=root))
   > Set PelvisRoot(PelvisRoot:double=MapRangeClamped(Value=BreakRotator(InRot=@n10.Rotation).Yaw, InRangeA=80, InRangeB=50, OutRangeA=1, OutRangeB=0)) -> (Output_Get:double)
   Sequence!(execute=@n1.then_2) -> (then_0, then_1, then_2)
   > n11 = Branch(Condition=Get bIsCrouching())
   > Sequence!(execute) -> (then_0, then_1)
   > Set jacketLegAlpha_L(jacketLegAlpha_L:double=MapRangeClamped(Value=Get legUp_L() -> (legUp_L:double), InRangeA=-100, InRangeB=0, OutRangeA=1, OutRangeB=0)) -> (Output_Get:double)
   Sequence(execute=@n11.else)
   > Set jacketLegAlpha_L(jacketLegAlpha_L=0)
   n12 = Get CurrentLOD()
   n13 = Get bRigidBodyEnabled()
   Set isRigidBodyDisabled(
    execute=@n1.then_0
    isRigidBodyDisabled:bool=OR Boolean(A=NOT Boolean(A=@n13), B="integer > integer"(A=@n12, B=1) {is=title K2Node_CallFunction})
   ) -> (Output_Get:bool)
   Event OnBeginCrouching()
   > Set isCrouching(isCrouching:double=1) -> (Output_Get:double)
   Event OnEndCrouching()
   > Set isCrouching(isCrouching=0)
   TryGetPawnOwner()
  ]}
]
variables [
 armFwd_R {type=number}
 headFwdBkRot {type=number}
 isCrouching {type=number}
 isRigidBodyDisabled {type=boolean}
 jacketLegAlpha_L {type=number}
 legUp_L {type=number}
 legUp_r {type=number}
 LegUpAndSpine {type=number}
 PelvisRoot {type=number}
 Spine_left {type=number}
 Spine_Right {type=number}
]
